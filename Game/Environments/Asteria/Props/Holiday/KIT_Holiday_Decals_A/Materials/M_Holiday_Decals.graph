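# Unlit deferred decal material for the Asteria holiday decal kit, using alpha-composite blending.
# Size and minimum U/V parameters scale and offset the texture coordinates, so a sub-rectangle of the textures can be selected, as with an atlas.
# Both the Diffuse and Mask textures sample those shared UVs, and their RGB values are multiplied together, then scaled by a Brightness parameter.
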Material /Game/Environments/Asteria/Props/Holiday/KIT_Holiday_Decals_A/Materials/M_Holiday_Decals
graph {
 properties {
  blendMode=BLEND_AlphaComposite
  materialDomain=MD_DeferredDecal
  pixelDepthOffsetMode=PDOM_Legacy
  shadingModel=MSM_Unlit
  usedShadingModels=Unlit
 }
 nodes [
  n1 = ScalarParameter() {defaultValue=1, parameterName=USize, sliderMax=1}
  n2 = ScalarParameter() {defaultValue=1, parameterName=VSize, sliderMax=1}
  n3 = Add(
   A=Multiply(A=TextureCoordinate(), B=AppendVector(A=@n1, B=@n2))
   B=AppendVector(A=ScalarParameter() {parameterName=UMin, sliderMax=1}, B=ScalarParameter() {parameterName=VMin, sliderMax=1})
  )
  Multiply(
   A=Multiply(
    A=TextureSampleParameter2D(UVs=@n3) {parameterName=Diffuse, texture=/Game/Environments/Asteria/Props/Holiday/KIT_Holiday_Decals_A/Textures/KIT_Holiday_Decals_A_Diffuse}.RGB
    B=TextureSampleParameter2D(UVs=@n3) {parameterName=Mask, texture=/Game/Environments/Asteria/Props/Holiday/KIT_Holiday_Decals_A/Textures/KIT_Holiday_Decals_A_Alpha}.RGB
   )
   B=ScalarParameter() {defaultValue=0.2, parameterName=Brightness}
  )
 ]
}
